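# Translucent, unlit, two-sided Niagara sprite and mesh material for a pulse in the Boomshot skydiving contrail.
# Radial UVs offset by the Sweep dynamic parameter sample a clamped line texture, using original-UV derivatives for mip selection.
# That sweep, broken up by a cloud texture, combines with particle alpha and a powered sphere mask for opacity.
# Particle color runs through an emissive-adjustment function, and a parameter blends SplineThicken world position offset against zero.
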Material /BRCosmetics/Athena/Effects/Particles/Skydiving/NS_Contrail_Boomshot/FX/M_Contrail_BoomShot_Pulse
graph {
 properties {
  bAutomaticallySetUsageInEditor=false
  bUsedWithNiagaraMeshParticles=true
  bUsedWithNiagaraSprites=true
  blendMode=BLEND_Translucent
  shadingModel=MSM_Unlit
  twoSided=true
  usedShadingModels=Unlit
 }
 nodes [
  LinearInterpolate(
   A=MaterialFunctionCall(
    WidthBase=ScalarParameter() {defaultValue=0.4, parameterName=SplineBase}
    WidthTip=ScalarParameter() {defaultValue=0.4, parameterName=SplineTip}
    UVs for Thickness=TextureCoordinate()
    Expand U or V UV Channel=StaticBool()
   ) {materialFunction=&SplineThicken}.WorldPositionOffset
   B=Constant()
   Alpha=ScalarParameter() {parameterName=Use_SplineThicken}
  )
  n1 = ParticleColor()
  MaterialFunctionCall(Emissive=@n1.RGB) {materialFunction=&MF_FX_EmissiveAdjustments}
  n2 = TextureCoordinate()
  n3 = Add(A=@n2, B=Constant2Vector() {g=-0.5, r=-0.5})
  n4 = TextureSampleParameter2D!(
   UVs=Add(
    A=MaterialFunctionCall(Vector or UVs=@n3) {materialFunction=&VectorToRadialValue}.Radial Coordinates
    B=DynamicParameter() {defaultValue=<rgba -1 1 1 1>, paramNames [Sweep,Param2,Param3,Param4]}.Sweep
   )
   "DDX(UVs)"=DDX(Value=@n2)
   "DDY(UVs)"=DDY(Value=@n2)
   Apply View MipBias
  ) {
   mipValueMode=TMVM_Derivative
   parameterName=SweepTexture
   samplerSource=SSM_Clamp_WorldGroupSettings
   texture=/BRCosmetics/Effects/Fort_Effects/Emotes/AnglePatch/T_Emote_AnglePatch_Line
  }
  n5 = Multiply(A=@n4.R, B=TextureSample(UVs=TextureCoordinate()) {texture=/Game/Effects/Fort_Effects/Textures/Tile/T_CloudScape_08}.RGB)
  n6 = TextureCoordinate()
  n7 = Constant2Vector() {g=0.5, r=0.5}
  n8 = Constant() {r=0.5}
  n9 = Constant() {r=0.1}
  Multiply(
   A=Multiply(A=@n1.A, B=LinearInterpolate(Alpha=@n5))
   B=Power(Base=SphereMask(A=@n6, B=@n7, Radius=@n8, Hardness=@n9), Exp=ScalarParameter() {defaultValue=2})
  )
 ]
}
