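# Unlit translucent material for Niagara mesh particles that draws a stylized smoke-cloud trail for the Boomshot skydiving contrail.
# A normal map lit by a sun-direction parameter blends shadow and highlight colors, mixing in screen-space Ben-Day dots.
# Rotated, panning UVs feed smoke textures; opacity combines alpha, dynamic-parameter erosion, particle alpha, depth fade and switchable fades.
# SplineThicken shapes trail width from dynamic base and tip parameters.
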Material /BRCosmetics/Athena/Effects/Particles/Skydiving/NS_Contrail_Boomshot/FX/M_Contrail_BoomShot_Spline_Cloud
graph {
 properties {
  bAutomaticallySetUsageInEditor=false
  bUsedWithNiagaraMeshParticles=true
  blendMode=BLEND_Translucent
  shadingModel=MSM_Unlit
  usedShadingModels=Unlit
 }
 nodes [
  Constant() {r=0.2}
  n1 = TextureSampleParameter2D!(UVs=NamedRerouteUsage(), Bias, Apply View MipBias) {
   group=Texture
   mipValueMode=TMVM_MipBias
   parameterName=Texture Normal
   samplerType=SAMPLERTYPE_Normal
   texture=/Game/Effects/Fort_Effects/Gameplay/SmokeDustExplosions/Textures/T_RibbonSmokeMasked_01
  }
  n2 = Normalize(VectorInput=@n1.RGB)
  n3 = DotProduct(
   A=Normalize(VectorInput=VectorParameter() {group=Shading, parameterName=SunDirectionVector, defaultValue=<rgba 0 28.0194 8.6704 1>}.RGB)
   B=@n2
  )
  n4 = Saturate(None=@n3)
  n5 = Power(
   Base=SmoothStep(
    Min=ScalarParameter() {defaultValue=0.2, group=Shading, parameterName=Smooth Step Min, sliderMax=1, sliderMin=-1, sortPriority=3}
    Max=ScalarParameter() {defaultValue=1, group=Shading, parameterName=Smooth Step Max, sliderMax=1, sortPriority=4}
    Value=Add(A=@n4) {constB=0}
   )
   Exp=ScalarParameter() {defaultValue=3, group=Shading, parameterName=Color Power, sliderMax=5, sortPriority=5}
  )
  If(
   A=@n5
   B=ScalarParameter() {defaultValue=0.7, group=Shading, parameterName=Shadow Threshold, sliderMax=0.1}
   "A > B"=Constant() {r=1}
   "A < B"=Constant()
  )
  LinearInterpolate(A=Constant(), B=Fresnel(Normal=@n2) {baseReflectFraction=0.05}, Alpha=Saturate(None=Multiply(A=@n3) {constB=-1}))
  n6 = Normalize(VectorInput=Subtract(A=CameraPositionWS(), B=ActorPositionWS()))
  n7 = Transform(None=VectorParameter() {group=SplineThicken, parameterName=Lookat-VectortoCompare, defaultValue=<rgba 0 0 1 1>}.RGB) {transformSourceType=TRANSFORMSOURCE_Local}
  n8 = Multiply(A=Abs(None=DotProduct(A=@n6, B=@n7)), B=ScalarParameter() {defaultValue=1, group=SplineThicken, parameterName=Lookat-Dot_Mult})
  n9 = ScalarParameter() {defaultValue=1, group=SplineThicken, parameterName=Lookat-Dot_Exp}
  LinearInterpolate(
   A=ScalarParameter() {defaultValue=6, group=SplineThicken, parameterName=Thickness-LookAway}
   B=ScalarParameter() {defaultValue=6, group=SplineThicken, parameterName=Thickness-LookAt}
   Alpha=Clamp(None=Power(Base=@n8, Exp=@n9))
  )
  n10 = DynamicParameter() {parameterIndex=1, paramNames [Tile X, Tile Y, SplineWidthBAse, SplineWidthTip]}
  n11 = ScalarParameter() {defaultValue=0.1, parameterName=Param_6}
  n12 = LinearInterpolate(
   A=ScalarParameter() {defaultValue=0.2, parameterName=Param_7}
   Alpha=OneMinus(None=MaterialFunctionCall!(Width=@n11, Sharpness, Offset, Direction Switch, Compare, Input Coordinates) {materialFunction=/Engine/Functions/Engine_MaterialFunctions03/Procedurals/GeneratedBand}.Result)
  ) {constA=0.1}
  n13 = Abs(None=Subtract(A=TextureCoordinate()) {constB=0.5})
  MaterialFunctionCall(
   WidthBase=LinearInterpolate(A=@n10.SplineWidthBAse, B=Multiply(A=@n12, B=@n10.SplineWidthBAse), Alpha=ScalarParameter() {defaultValue=1, parameterName=UseWidthMask}) {constA=1}
   WidthTip=@n10.SplineWidthTip
   UVs for Thickness=OneMinus(None=Multiply(A=@n13) {constB=2})
  ) {materialFunction=&SplineThicken}
  n14 = TextureSampleParameter2D(UVs=NamedRerouteUsage()) {
   group=Texture
   mipValueMode=TMVM_MipBias
   parameterName=Texture Alpha
   texture=/Game/Effects/Fort_Effects/Gameplay/SmokeDustExplosions/Textures/T_RibbonSmokeMasked_02
  }
  n15 = OneMinus(None=ComponentMask(None=TextureCoordinate()) {g=true})
  Min(A=@n14.R, B=@n15)
  n16 = ParticleColor()
  n17 = TextureSampleParameter2D(UVs=NamedRerouteUsage()) {
   group=Texture
   mipValueMode=TMVM_MipBias
   parameterName=Texture Erosion
   samplerType=SAMPLERTYPE_LinearColor
   texture=/Game/Effects/Fort_Effects/Gameplay/SmokeDustExplosions/Textures/T_RibbonSmokeMasked_03
  }
  n18 = TextureCoordinate()
  n19 = Multiply(A=@n17.R, B=OneMinus(None=ComponentMask(None=@n18) {r=true}))
  n20 = DynamicParameter() {defaultValue=<rgba 0 0 0.1 1>, paramNames [Erosion, Random X, Leading Color Swipe, Param4]}
  n21 = ScalarParameter() {defaultValue=5, group=Shading, parameterName=Erosion Contrast, sliderMax=20}
  n22 = Multiply(
   A=@n16.A
   B=Multiply(
    A=@n14.R
    B=MaterialFunctionCall!(Ramp=@n19, Dissolve=@n20.Erosion, EdgeSharpness=@n21) {materialFunction=/Game/Effects/Fort_Effects/Materials/_Functions/MF_Dissolve}.ClampedResult
   )
  )
  n23 = ScalarParameter() {defaultValue=0.2, parameterName=Fade Start, sliderMax=1}
  n24 = ScalarParameter() {defaultValue=0.4, parameterName=Fade End, sliderMax=1}
  n25 = ComponentMask(None=TextureCoordinate()) {r=true}
  Multiply(
   A=@n22
   B=StaticSwitchParameter(True=SmoothStep(Min=@n23, Max=@n24, Value=@n25) {constMax=0.4, constMin=0.2}, False=Constant() {r=1}) {parameterName=Add Front Fade}
  )
  Multiply(
   A=MaterialFunctionCall(Width=ScalarParameter() {defaultValue=0.5, parameterName=Param_4}, Sharpness=ScalarParameter() {defaultValue=0.9, parameterName=Param_5}) {materialFunction=&GeneratedBand}.Result
   B=@n14.R
  )
  n26 = ComponentMask(None=TextureCoordinate()) {r=true}
  n27 = SmoothStep(Min=ScalarParameter() {defaultValue=0.4}, Max=ScalarParameter() {defaultValue=1, parameterName=Param_2}, Value=@n26)
  n28 = ScalarParameter() {defaultValue=99, parameterName=Param_3}
  n29 = ScalarParameter() {defaultValue=0.1, parameterName=ViewFadeMin, sliderMax=1}
  n30 = ScalarParameter() {defaultValue=0.2, parameterName=ViewFadeMax, sliderMax=1}
  n31 = Abs(None=DotProduct(A=CameraVectorWS(), B=VertexNormalWS()))
  n32 = Multiply(
   A=DepthFade(Opacity=@n22, FadeDistance=ScalarParameter() {defaultValue=30, group=Shading, parameterName=Depth Fade Distance, sliderMax=100})
   B=StaticSwitchParameter(True=SmoothStep(Min=@n29, Max=@n30, Value=@n31) {constMax=0.1}, False=Constant() {r=1}) {parameterName=UseViewAngleFade}
  )
  Multiply(A=Saturate(None=Multiply(A=@n27, B=@n28) {constB=22}), B=@n32)
  n33 = MaterialFunctionCall(UVs=TextureCoordinate(), Rotation Center=Constant2Vector() {g=-0.5, r=0.5}, "Rotation Angle (0-1)"=Constant() {r=0.25}) {materialFunction=&CustomRotator}
  n34 = DynamicParameter() {parameterIndex=1, paramNames [Tile X, Tile Y, SplineWidthBAse, SplineWidthTip]}
  n35 = AppendVector(A=@n34.Tile X, B=@n34.Tile X)
  n36 = ScalarParameter() {defaultValue=0.2, group=Texture, parameterName=Speed X, sliderMax=5}
  n37 = Constant()
  NamedRerouteDeclaration(None=Panner(Coordinate=Multiply(A=@n33.Rotated Values, B=@n35), Speed=AppendVector(A=@n36, B=@n37))) {name=UV, nodeColor=<rgba 0 0.2266 1 1>}
  OneMinus(None=@n26)
  Power(Base=@n15) {constExponent=3}
  n38 = VectorParameter() {group=Shading, parameterName=ShadowColor, sortPriority=0, defaultValue=<rgba 0.8288 0.0531 0.6303 1>}
  n39 = VectorParameter() {group=Shading, parameterName=HighLightColor, sortPriority=2, defaultValue=<rgba 3 0.0048 2.7589 1>}
  n40 = ScreenPosition()
  n41 = MaterialFunctionCall() {materialFunction=&ScreenResolution}
  n42 = ComponentMask(None=@n41.Visible Resolution) {r=true}
  n43 = ComponentMask(None=@n41.Visible Resolution) {g=true}
  n44 = ScalarParameter() {defaultValue=150, group=Shading, parameterName=Dots Tiling, sliderMax=100}
  n45 = AppendVector(A=Multiply(A=Divide(A=@n42, B=@n43), B=@n44), B=@n44)
  n46 = Distance(A=Frac(None=Multiply(A=@n40.ViewportUV, B=@n45) {constB=50}), B=Constant() {r=0.5})
  n47 = Multiply(A=ScalarParameter() {defaultValue=0.5, group=Shading, parameterName=Dot Size, sliderMax=0.5}, B=OneMinus(None=@n4))
  n48 = Constant()
  n49 = Constant() {r=1}
  n50 = LinearInterpolate(
   A=@n38.RGB
   B=LinearInterpolate(
    A=@n39.RGB
    B=VectorParameter() {parameterName=BenDayDotColor, defaultValue=<rgba 0.3686 0.2153 0.3279 1>}.RGB
    Alpha=If(A=@n46, B=@n47, "A > B"=@n48, "A < B"=@n49) {constB=0.2}
   )
   Alpha=@n4
  )
  Reroute(None=MaterialFunctionCall(Emissive=Multiply(A=@n50, B=@n16.RGB)) {materialFunction=&MF_FX_EmissiveAdjustments}.Result)
  n51 = MaterialFunctionCall(Fade Length=ScalarParameter() {defaultValue=256, parameterName=Near Camera Fade Distance, sliderMax=1024}) {materialFunction=&CameraDepthFade}
  n52 = Constant() {r=1}
  Reroute(None=Multiply(A=@n32, B=StaticSwitchParameter(True=@n51.Result, False=@n52) {parameterName=UseNearCameraFade}))
  Round(None=SmoothStep(Value=@n5) {constMin=0.5})
  Round(None=@n15)
  ScalarParameter() {defaultValue=0.4, parameterName=Param_1}
  ScalarParameter() {defaultValue=1, group=Texture, parameterName=Tile X, sliderMax=5}
  ScalarParameter() {defaultValue=1, group=Texture, parameterName=Tile Y, sliderMax=5}
  SmoothStep(Value=@n4) {constMax=0.6, constMin=0.5}
  Transform(None=@n1.RGB)
  VectorParameter() {defaultValue=<rgba -0.6411 0 1 1>}
 ]
}
