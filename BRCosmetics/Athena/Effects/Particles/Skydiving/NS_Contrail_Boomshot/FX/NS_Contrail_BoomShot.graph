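# Niagara skydiving contrail "BoomShot": firing bombs with fire and wind cones that home toward points near the player and burst into heart-themed explosions.
# Bomb_Fire001 drives Hearts_Trail and Smoke_Ribbon through particle attribute readers.
# Bombs write deaths to a system position array; ReadDeath supplies the trigger and position for explosion, cloud, heart and smoke bursts.
# Front-end modifiers give separate menu and gameplay offsets, sizes and speeds.
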NiagaraSystem /BRCosmetics/Athena/Effects/Particles/Skydiving/NS_Contrail_Boomshot/FX/NS_Contrail_BoomShot
dependencies {
 usedDataInterfaces [
  &NiagaraDataInterfaceArrayFloat3
  &NiagaraDataInterfaceCurve
  &NiagaraDataInterfaceParticleRead
  &NiagaraDataInterfaceColorCurve
  &NiagaraDataInterfaceArrayFloat
  &NiagaraDataInterfaceVectorCurve
  &NiagaraDataInterfaceArrayPosition
 ]
 usedDynamicInputs [
  &SimulationPosition
  &Multiply_Vector2
  &MakeVector2D
  &UniformAOrBFloat
  &UniformRangedLinearColor
  &MaskFloatBySpawnGroup
  &NormalizedAngleToDegrees
 ]
 usedModules [
  &self:SystemSpawnScript.SystemScriptSource.SystemScriptGraph.NiagaraNodeAssignment_2.SetVariables_#1
  &SystemState
  /Game/Effects/Niagara/Modules/Contrails/FN_Contrail_Setup
  &FN_Integrator
  &self:ReadDeath
  &self:SystemSpawnScript.SystemScriptSource.SystemScriptGraph.NiagaraNodeAssignment_3.SetVariables_#2
 ]
}
emitters {
 Bomb_Fire001 {
  emitterUpdateScript [
   &EmitterState(Life Cycle Mode=::NewEnumerator1, Loop Duration=3, UseLoopDelay=true)
   &SpawnBurst_Instantaneous(Use Spawn Probability=false, UseLoopCountLimit=false) {module=SpawnBurst_Instantaneous001}
   &SpawnRate(Spawn Probability=0.05, SpawnRate=3, Use Spawn Probability=true) {module=SpawnRate001}
  ]
  particleSpawnScript [
   &PartitionParticles(Number Of Partitions=3)
   &InitializeParticle(
    Color Mode=::NewEnumerator0
    Lifetime Mode=::NewEnumerator1
    Mesh Scale Mode=::NewEnumerator3
    UsePositionOffset=false
    Write Material Random=true
   ) {module=InitializeParticle001}
   &SystemLocation(
    MaskSpawnGroup=false
    Offset=&FN_FrontEndVector3Modifier(
     Gameplay=&SelectVectorFromArray(
      Array Sampling Mode=::NewEnumerator0
      Direct Array Index=?
      Vector Selection Array=DataInterface() {propertyValues {FloatData [<400 -150 350>, <-500 -150 350>, <-500 -150 350>]}}
      Random Seed=0
     )
     Front End=<0 0 350>
    )
   ) {module=SystemLocation001}
   &ShapeLocation(MaskSpawnGroup=false, Non Uniform Scale=<1 1 0.5>, Override Seed=false, UseOwnerScale=true)
   &InitialMeshOrientation(Mesh Axis To Orient=<0 0 -1>, Mesh Orientation Mode=::NewEnumerator2, Use Rotation Vector=false) {module=InitialMeshOrientation001}
  ]
  particleUpdateScript [
   &ParticleState()
   &SolveForcesAndVelocity(Clamp Velocity=false, Limit Acceleration=false)
   /SetVariables(Particles.ImpactPosition:NiagaraPosition=&FN_FrontEndVector3Modifier(
     Gameplay=&AddVectorToPosition(
      Position=$Engine.Emitter.SimulationPosition
      Vector=&RandomRangeVector(Minimum=<0 290 -200>, Maximum=<70 300 -200>)
     )
     Front End=&AddVectorToPosition(Position=$Engine.Emitter.SimulationPosition, Vector=<0 200 -150>)
    )) {isSetParameters=true}
   /SetVariables(Particles.Position=&Lerp_Vector(
     A=Linked() {linkedVariable {name=Particles.Position, type=NiagaraPosition}}
     B=Linked() {linkedVariable {name=Particles.ImpactPosition, type=NiagaraPosition}}
     Alpha=&FloatFromCurve(FloatCurve=DataInterface() {propertyValues {Curve {keys [
          {interpMode=RCIM_Cubic, tangentMode=RCTM_User, time=0, value=0}
          {arriveTangent=0.5165, interpMode=RCIM_Cubic, leaveTangent=0.5165, tangentMode=RCTM_User, time=1, value=0.3}
         ]}}})
    )) {isSetParameters=true}
   /SetVariables(Particles.Velocity=&Multiply_VectorByFloat(
     Vector=&SubtractVectorFromPosition(Position=$Particles.Position, Vector=Linked() {linkedVariable {name=Particles.Previous.Position, type=NiagaraPosition}})
     Float=1
    )) {isSetParameters=true}
   /SetVariables(Particles.Jitter:NiagaraFloat=&Waveform(
     "Waveform [1]"=::NewEnumerator0
     "[1] Phase Scale"=&RandomRangeFloat(Minimum=11, Maximum=15)
     "[1] Frequency"=1
     "[1] Amplitude Min/Max"=<0.6 1>
    )) {isSetParameters=true}
   /SetVariables(
    Particles.FireConeColor:LinearColor=&MakeLinearColor(
     R:NiagaraFloat=1
     G:NiagaraFloat=1
     B:NiagaraFloat=1
     A:NiagaraFloat=&FloatFromCurve(FloatCurve=DataInterface() {propertyValues {Curve {keys [
          {interpMode=RCIM_Cubic, leaveTangent=5.4742, tangentMode=RCTM_Break, time=0.11, value=0}
          {arriveTangent=1.4286, interpMode=RCIM_Cubic, leaveTangent=0.3017, tangentMode=RCTM_Break, time=0.2, value=1}
          {interpMode=RCIM_Cubic, tangentMode=RCTM_User, time=0.8, value=1}
          {interpMode=RCIM_Cubic, tangentMode=RCTM_User, time=0.85, value=0}
         ]}}})
    )
    Particles.FireConePosition:NiagaraPosition=&AddVectorToPosition(
     Position=$Particles.Position
     Vector=&Multiply_VectorByFloat(
      Vector=&NormalizeVector(Vector To Normalize=$Particles.Velocity, Safe Normalize=false, Safe Fallback Threshold=0, Safe Fallback Vector=<0 0 0.001>)
      Float=-65
     )
    )
    Particles.FireConeScale:Vector3f=&Multiply_VectorByFloat(
     Vector=&MakeVector(
      X=1
      Y=1
      Z=&Multiply_Float(
       A=&Multiply_Float(
        A=9
        B=&FloatFromCurve(FloatCurve=DataInterface() {propertyValues {Curve {keys [
             {interpMode=RCIM_Cubic, leaveTangent=-17.953, tangentMode=RCTM_Break, time=0.8, value=1}
             {interpMode=RCIM_Cubic, tangentMode=RCTM_Break, time=0.85, value=0}
            ]}}})
       )
       B=$Particles.Jitter
      )
     )
     Float=&FloatFromCurve(
      FloatCurve=DataInterface() {propertyValues {Curve {keys [
          <at 0.11 0 RCIM_Cubic>
          {arriveTangent=-0.1738, interpMode=RCIM_Cubic, leaveTangent=-0.1738, tangentMode=RCTM_User, time=0.2, value=1}
          {interpMode=RCIM_Cubic, tangentMode=RCTM_User, time=0.8, value=1}
          {interpMode=RCIM_Cubic, tangentMode=RCTM_User, time=1, value=0}
         ]}}}
      Scale Curve=0.2
     )
    )
   ) {isSetParameters=true}
   /SetVariables(
    Particles.WindConeColor:LinearColor=&MakeLinearColorFromVectorAndFloat(
     "Vector (RGB)"=<1 1 1>
     "Float (Alpha)"=&FloatFromCurve(FloatCurve=DataInterface() {propertyValues {Curve {keys [
          {interpMode=RCIM_Cubic, tangentMode=RCTM_User, time=0.13, value=0}
          {arriveTangent=6.9757, interpMode=RCIM_Cubic, leaveTangent=-4.8624, tangentMode=RCTM_Break, time=0.3, value=1}
          {interpMode=RCIM_Cubic, tangentMode=RCTM_Break, time=0.9, value=0}
         ]}}})
    )
    Particles.WindConePosition:NiagaraPosition=&AddVectorToPosition(
     Position=$Particles.Position
     Vector=&Multiply_VectorByFloat(
      Vector=&NormalizeVector(Vector To Normalize=$Particles.Velocity, Safe Normalize=false, Safe Fallback Threshold=0, Safe Fallback Vector=<0 0 0.001>)
      Float=20
     )
    )
    Particles.WindConeScale:Vector3f=&Multiply_VectorByFloat(
     Vector=&RandomRangeVector(Minimum=<1 0.7 0.7>, Maximum=<0.6 0.8 0.8>)
     Float=&FloatFromCurve(FloatCurve=DataInterface() {propertyValues {Curve {keys [
          <at 0 0 RCIM_Cubic>
          {arriveTangent=-0.1738, interpMode=RCIM_Cubic, leaveTangent=-0.1738, tangentMode=RCTM_User, time=0.1, value=1}
          {interpMode=RCIM_Cubic, tangentMode=RCTM_User, time=0.8, value=1}
          {interpMode=RCIM_Cubic, tangentMode=RCTM_User, time=1, value=0}
         ]}}})
    )
   ) {isSetParameters=true}
   &ScaleMeshSize(Scale Factor=&VectorFromFloat(Value=&FloatFromCurve(FloatCurve=DataInterface() {propertyValues {Curve {keys [
          <at 0 0 RCIM_Cubic>
          {arriveTangent=-0.1738, interpMode=RCIM_Cubic, leaveTangent=-0.1738, tangentMode=RCTM_User, time=0.1, value=1}
          {interpMode=RCIM_Cubic, tangentMode=RCTM_User, time=0.8, value=1}
          {interpMode=RCIM_Cubic, tangentMode=RCTM_User, time=1, value=0}
         ]}}})))
   &self:WriteDeath(Position Array:NiagaraDataInterfaceArrayPosition=$System.DeathArray) {module=WriteDeath}
  ]
  rendererClasses [&NiagaraMeshRendererProperties]
  renderers [
   {rendererClass=&NiagaraMeshRendererProperties, rendererIndex=0}
   {rendererClass=&NiagaraMeshRendererProperties, rendererIndex=1}
   {rendererClass=&NiagaraMeshRendererProperties, rendererIndex=2}
  ]
 }
 Clouds_Spline {
  emitterUpdateScript [
   &EmitterState()
   &SpawnPerFrame(Spawn:NiagaraBool=$System.DeathSpawn, Spawn Count=2, Use Spawn Probability=false)
   &SpawnPerFrame(Spawn:NiagaraBool=$System.DeathSpawn, Spawn Count=2, Spawn Group=1, Use Spawn Probability=false) {module=SpawnPerFrame001}
  ]
  particleSpawnScript [
   &PartitionParticles(Number Of Partitions=$Emitter.SpawnGroup)
   &InitializeParticle(
    Color Mode=::NewEnumerator0
    Lifetime=1.5
    Mesh Scale=&Lerp_Vector(A=<0.5 0.5 0.5>, B=<1 1 1>, Alpha=&MakeFloatFromInt(Int=$Emitter.SpawnGroup))
    Mesh Scale Mode=::NewEnumerator4
    UsePositionOffset=false
    Write Material Random=true
   )
   /SetVariables(Particles.Position=$System.DeathPosition) {isSetParameters=true}
   &ShapeLocation(
    MaskSpawnGroup=false
    Override Seed=false
    Ring / Disc Distribution Mode:ENiagaraSphereDistributionMode=::NewEnumerator1
    Ring Radius=1
    Shape Primitive=::NewEnumerator4
    U Position=&ReturnNormalizedExecIndex(Normalized Index Includes One=false)
    UseOwnerScale=true
   )
   &AddVelocityFromPoint(
    Module.Use Falloff Distance=false
    Origin Offset Coordinate Space=::Simulation
    Velocity Origin:NiagaraPosition=$System.DeathPosition
    Velocity Strength=&Lerp_Float(A=22, B=99, Alpha=&MakeFloatFromInt(Int=$Emitter.SpawnGroup))
   )
   &ApplyInitialForces()
  ]
  particleUpdateScript [
   &ParticleState()
   &ScaleMeshSize(Scale Factor=&Multiply_Vector(
     A=&MakeVector(
      X=&FloatFromCurve(
       FloatCurve=DataInterface() {propertyValues {Curve {keys [
           {interpMode=RCIM_Cubic, tangentMode=RCTM_User, time=0, value=0.8}
           {arriveTangent=1.9213, interpMode=RCIM_Cubic, leaveTangent=1.9213, tangentMode=RCTM_User, time=1, value=1.5}
          ]}}}
       Scale Curve=&RandomRangeFloat(Minimum=1, Maximum=1.3)
      )
      Y=1
      Z=1
     )
     B=&VectorFromFloat(Value=&FloatFromCurve(FloatCurve=DataInterface() {propertyValues {Curve {keys [
           {arriveTangent=7.9207, interpMode=RCIM_Cubic, leaveTangent=5.2162, tangentMode=RCTM_User, time=0, value=0.15}
           {arriveTangent=1.0025, interpMode=RCIM_Cubic, leaveTangent=1.0025, tangentMode=RCTM_User, time=0.2012, value=0.7293}
           {arriveTangent=0.2754, interpMode=RCIM_Cubic, leaveTangent=0.2754, tangentMode=RCTM_User, time=1.0463, value=1}
          ]}}}))
    )) {module=ScaleMeshSize002}
   &DynamicMaterialParameters(
    Index 0 Param 1=&FloatFromCurve(
     FloatCurve=DataInterface() {propertyValues {Curve {keys [
         {interpMode=RCIM_Cubic, leaveTangent=-7.5235, tangentMode=RCTM_Break, time=0, value=1}
         {arriveTangent=0.4054, interpMode=RCIM_Cubic, leaveTangent=2.1589, tangentMode=RCTM_Break, time=0.3, value=0}
         <at 1 1 RCIM_Cubic>
        ]}}}
     Scale Curve=&Lerp_Float(A=1, B=1, Alpha=&MakeFloatFromInt(Int=$Emitter.SpawnGroup))
    )
    Index 1 Param 1=&RandomRangeFloat(Minimum=0.5, Maximum=1)
    Index 1 Param 2=1
    Index 1 Param 3=&Multiply_Float(
     A=66
     B=&FloatFromCurve(FloatCurve=DataInterface() {propertyValues {Curve {keys [
          {interpMode=RCIM_Cubic, tangentMode=RCTM_User, time=0, value=0.3}
          {arriveTangent=6.1311, interpMode=RCIM_Cubic, leaveTangent=6.1311, tangentMode=RCTM_User, time=0.3, value=1}
         ]}}})
    )
    Index 1 Param 4=&RandomRangeFloat(Minimum=1, Maximum=1)
    Module.Index0Param0Write=true
    Module.Index0Param1Write=true
    Module.Index0Param2Write=true
    Module.Index0Param3Write=true
    Module.Index1Param0Write=true
    Module.Index1Param1Write=true
    Module.Index1Param2Write=true
    Module.Index1Param3Write=true
    Write Parameter Index 1=true
   )
   &Drag(
    Drag=&FloatFromCurve(
     FloatCurve=DataInterface() {propertyValues {Curve {keys [{interpMode=RCIM_Cubic, tangentMode=RCTM_User, time=0, value=0}, {interpMode=RCIM_Cubic, tangentMode=RCTM_User, time=1, value=1}]}}}
     Scale Curve=4
    )
    Use Linear Drag=true
    Use Rotational Drag=true
   ) {module=Drag002}
   &AccelerationForce(Acceleration=&Multiply_VectorByFloat(
     Vector=&RandomRangeVector(Minimum=<-0.2 -0.2 0>, Maximum=<0.2 0.2 0.2>)
     Float=&Lerp_Float(A=55, B=55, Alpha=&MakeFloatFromInt(Int=$Emitter.SpawnGroup))
    )) {module=AccelerationForce001}
   &InheritSourceMovement(
    Applied Position Scale=&MakeVector(
     X=0
     Y=0
     Z=&FloatFromCurve(
      FloatCurve=DataInterface() {propertyValues {Curve {keys [{interpMode=RCIM_Cubic, tangentMode=RCTM_User, time=0, value=1}, {interpMode=RCIM_Cubic, tangentMode=RCTM_User, time=1, value=0}]}}}
      Scale Curve=-0.1
     )
    )
    ApplyForce=false
    ApplyPosition=true
    ApplyPositionFalloff=false
   ) {module=InheritSourceMovement001}
   &SolveForcesAndVelocity(Clamp Velocity=false, Limit Acceleration=false)
   &ScaleColor(
    Scale RGB=&MakeVectorFromLinearColorRGB(Color=&ColorFromCurve(ColorCurve=DataInterface() {propertyValues {
        AlphaCurve {keys [<at 0 1 RCIM_Cubic>,<at 1 1 RCIM_Cubic>]}
        BlueCurve {keys [<at 0 1 RCIM_Cubic>,<at 0.395 1>,<at 0.5833 0.3549>]}
        GreenCurve {keys [<at 0 1 RCIM_Cubic>,<at 0.395 1>,<at 0.5833 0.1487>]}
        RedCurve {keys [<at 0 1 RCIM_Cubic>,<at 0.395 1>,<at 0.5833 0.2341>]}
       }}))
    ScaleA=true
    ScaleRGB=true
   )
  ]
  rendererClasses [&NiagaraMeshRendererProperties]
  renderers [{rendererClass=&NiagaraMeshRendererProperties, rendererIndex=0}]
 }
 Explosion {
  emitterUpdateScript [
   &EmitterState()
   &SpawnPerFrame(Spawn:NiagaraBool=$System.DeathSpawn, Use Spawn Probability=false)
  ]
  particleSpawnScript [
   &InitializeParticle(
    Color=<rgba 2 0.4851 0.2053 1>
    Lifetime=0.4
    Mesh Scale Mode=::NewEnumerator3
    Mesh Uniform Scale=3
    UsePositionOffset=false
    Write Material Random=true
   )
   /SetVariables(Particles.BenDayLifetime=0.3, Particles.WaveLifetime=1.5) {isSetParameters=true}
   &InitialMeshOrientation(
    Rotation=&RandomRangeVector(Minimum=<-180 -180 -180>, Maximum=<180 180 180>)
    Use Orientation Vector=true
    Use Rotation Vector=true
   )
   /SetVariables(Particles.Position=$System.DeathPosition) {isSetParameters=true}
   &SpriteFacingAndAlignment(Sprite Facing=<0 0 1>, WriteAlignment=false, WriteFacing=true)
  ]
  particleUpdateScript [
   &ParticleState()
   &ScaleColor(
    Scale Alpha=&FloatFromCurve(FloatCurve=DataInterface() {propertyValues {Curve {keys [
         {tangentMode=RCTM_User, time=0, value=1}
         {arriveTangent=-0.0405, interpMode=RCIM_Cubic, leaveTangent=-0.0405, tangentMode=RCTM_User, time=0.7009, value=1}
         {arriveTangent=-8.0089, interpMode=RCIM_Cubic, leaveTangent=-8.0089, tangentMode=RCTM_User, time=1, value=0}
        ]}}})
    Scale RGB=&VectorFromFloat(Value=&FloatFromCurve(
      FloatCurve=DataInterface() {propertyValues {Curve {keys [{interpMode=RCIM_Cubic, tangentMode=RCTM_User, time=0, value=1}, {interpMode=RCIM_Cubic, tangentMode=RCTM_User, time=1, value=0}]}}}
      Scale Curve=20
     ))
    ScaleA=true
    ScaleRGB=true
   )
   &DynamicMaterialParameters(
    Index 0 Param 1=&FloatFromCurve(FloatCurve=DataInterface() {propertyValues {Curve {keys [{arriveTangent=0.0592, interpMode=RCIM_Cubic, time=0, value=0.1}, <at 1 1 RCIM_Cubic>]}}})
    Index 0 Param 2=50
    Index 0 Param 3=1
    Index 0 Param 4=1
    Module.Index0Param0Write=true
    Module.Index0Param1Write=true
    Module.Index0Param2Write=true
    Module.Index0Param3Write=true
   )
   /SetVariables(
    Particles.BenDayUpdateAge:NiagaraFloat=&Divide_Float(A=$Particles.Age, B=$Particles.Initial.BenDayLifetime)
    Particles.WaveUpdateAge:NiagaraFloat=&Divide_Float(A=$Particles.Age, B=$Particles.Initial.WaveLifetime)
   ) {isSetParameters=true}
   /SetVariables(
    Particles.GlowColor=&MakeLinearColorFromVectorAndFloat(
     "Vector (RGB)"=&Multiply_Vector(
      A=&Multiply_Vector(
       A=&MakeVectorFromLinearColorRGB(Color=<rgba 4 1.2169 0.2977 1>)
       B=&VectorFromCurve(VectorCurve=DataInterface() {propertyValues {
          XCurve {keys [<at 0 1>,<at 0.25 0.5>]}
          YCurve {keys [<at 0 1>,<at 0.25 0.5>,<at 0.5 0.5>,<at 1 0.5>]}
          ZCurve {keys [<at 0 1>,<at 0.25 0.5>]}
         }})
      )
      B=&VectorFromFloat(Value=&FloatFromCurve(FloatCurve=DataInterface() {propertyValues {Curve {keys [
            <at 0 0 RCIM_Cubic>
            {arriveTangent=0.3223, interpMode=RCIM_Cubic, leaveTangent=0.3223, tangentMode=RCTM_User, time=0.1, value=1}
            {arriveTangent=0.0367, interpMode=RCIM_Cubic, leaveTangent=-3.7949, tangentMode=RCTM_Break, time=0.3, value=1}
            <at 1.0011 0.002 RCIM_Cubic>
           ]}}}))
     )
     "Float (Alpha)"=1
    )
    Particles.GlowScale:Vector3f=&VectorFromFloat(Value=3)
   ) {isSetParameters=true}
   /SetVariables(
    Particles.BenDayCameraOffset=25
    Particles.BenDayColor:LinearColor=&MultiplyLinearColor(
     Linear Color=&ColorFromCurve(
      ColorCurve=DataInterface() {propertyValues {
        AlphaCurve {keys [<at 0.4073 1 RCIM_Cubic>]}
        BlueCurve {keys [<at 0.2311 0.045 RCIM_Cubic>, {interpMode=RCIM_Cubic, leaveTangent=0.0649, time=0.5283, value=0.1937}, <at 0.6851 0.0833>]}
        GreenCurve {keys [<at 0.2311 0.6513 RCIM_Cubic>, {interpMode=RCIM_Cubic, leaveTangent=-1.1021, time=0.5283, value=0.075}, <at 0.6851 0>]}
        RedCurve {keys [<at 0.2311 1.5 RCIM_Cubic>, {interpMode=RCIM_Cubic, leaveTangent=-0.8461, time=0.5283, value=1.5}, <at 0.6851 1>]}
       }}
      CurveIndex=$Particles.BenDayUpdateAge
     )
     Scale Factor=&MakeVector4(
      X=1
      Y=1
      Z=1
      W=&FloatFromCurve(
       FloatCurve=DataInterface() {propertyValues {Curve {keys [
           {interpMode=RCIM_Cubic, tangentMode=RCTM_User, time=0.3, value=1}
           {arriveTangent=-2.5266, interpMode=RCIM_Cubic, leaveTangent=-2.5266, tangentMode=RCTM_User, time=1, value=0}
          ]}}}
       CurveIndex=$Particles.BenDayUpdateAge
       Scale Curve=&FNTimeOfDayScalar(Night Scale=1, Morning Scale=1, Evening Scale=1)
      )
     )
    )
    Particles.BenDayDMP:Vector4f=&MakeVector4(
     X=&FloatFromCurve(
      FloatCurve=DataInterface() {propertyValues {Curve {keys [
          {arriveTangent=0.89, interpMode=RCIM_Cubic, leaveTangent=0.89, tangentMode=RCTM_User, time=0, value=0}
          {interpMode=RCIM_Cubic, tangentMode=RCTM_User, time=1, value=1}
         ]}}}
      CurveIndex=$Particles.BenDayUpdateAge
      Scale Curve=0.4
     )
     Y=&FloatFromCurve(
      FloatCurve=DataInterface() {propertyValues {Curve {keys [
          {arriveTangent=-0.0405, interpMode=RCIM_Cubic, leaveTangent=-0.0405, tangentMode=RCTM_User, time=0.5, value=1}
          {arriveTangent=-0.0719, interpMode=RCIM_Cubic, leaveTangent=-0.0719, tangentMode=RCTM_User, time=1, value=0.2}
         ]}}}
      CurveIndex=$Particles.BenDayUpdateAge
      Scale Curve=0.25
     )
     Z=0
     W=0
    )
    Particles.BenDayScale:Vector2f=&Vector2DFromFloat(Value=&FloatFromCurve(
      FloatCurve=DataInterface() {propertyValues {Curve {keys [
          {interpMode=RCIM_Cubic, tangentMode=RCTM_User, time=0, value=1}
          <at 0.2 3 RCIM_Cubic>
          {interpMode=RCIM_Cubic, tangentMode=RCTM_User, time=1, value=1}
         ]}}}
      CurveIndex=$Particles.BenDayUpdateAge
      Scale Curve=150
     ))
   ) {isSetParameters=true}
   /SetVariables(
    Particles.WaveColor:LinearColor=&MultiplyLinearColor(
     Linear Color=<rgba 3 0.6439 2.2614 1>
     Scale Factor=&MakeVector4(
      X=1
      Y=1
      Z=1
      W=&FloatFromCurve(
       FloatCurve=DataInterface() {propertyValues {Curve {keys [
           {arriveTangent=-0.024, interpMode=RCIM_Cubic, leaveTangent=-0.024, tangentMode=RCTM_User, time=0, value=0}
           {arriveTangent=-1.4602, interpMode=RCIM_Cubic, leaveTangent=-1.4602, tangentMode=RCTM_User, time=0.1, value=1}
           {arriveTangent=-1.1111, interpMode=RCIM_Cubic, leaveTangent=-1.1111, tangentMode=RCTM_Break, time=0.8, value=0.1}
           <at 1 0 RCIM_Cubic>
          ]}}}
       CurveIndex=$Particles.WaveUpdateAge
       Scale Curve=11
      )
     )
    )
    Particles.WaveDMP:Vector4f=&MakeVector4(
     X=&FloatFromCurve(
      FloatCurve=DataInterface() {propertyValues {Curve {keys [
          {arriveTangent=-4.087, interpMode=RCIM_Cubic, leaveTangent=-4.087, tangentMode=RCTM_User, time=0, value=0.5}
          {arriveTangent=-1.721, interpMode=RCIM_Cubic, leaveTangent=-1.721, tangentMode=RCTM_User, time=0.1324, value=-0.0129}
          {arriveTangent=-0.376, interpMode=RCIM_Cubic, leaveTangent=-0.376, tangentMode=RCTM_User, time=1, value=-1}
         ]}}}
      CurveIndex=$Particles.WaveUpdateAge
     )
     Y=0
     Z=0
     W=0
    )
    Particles.WaveScale:Vector2f=&Vector2DFromFloat(Value=250)
   ) {isSetParameters=true}
   &InheritSourceMovement(
    Applied Position Scale=&MakeVector(
     X=0
     Y=0
     Z=&FloatFromCurve(
      FloatCurve=DataInterface() {propertyValues {Curve {keys [{interpMode=RCIM_Cubic, tangentMode=RCTM_User, time=0, value=1}, {interpMode=RCIM_Cubic, tangentMode=RCTM_User, time=1, value=0}]}}}
      Scale Curve=-0.1
     )
    )
    ApplyForce=false
    ApplyPosition=true
    ApplyPositionFalloff=false
   ) {module=InheritSourceMovement001}
   &SolveForcesAndVelocity(Clamp Velocity=false, Limit Acceleration=false)
   &ScaleMeshSize(Scale Factor=&VectorFromFloat(Value=&FloatFromCurve(
      FloatCurve=DataInterface() {propertyValues {Curve {keys [
          {arriveTangent=8.8659, interpMode=RCIM_Cubic, leaveTangent=8.8659, tangentMode=RCTM_User, time=0, value=0.25}
          {arriveTangent=0.75, interpMode=RCIM_Cubic, leaveTangent=0.75, time=0.1325, value=0.75}
          <at 1 1>
         ]}}}
      Scale Curve=2.4
     )))
  ]
  rendererClasses [&NiagaraMeshRendererProperties, &NiagaraSpriteRendererProperties]
  renderers [
   {rendererClass=&NiagaraMeshRendererProperties, rendererIndex=0}
   {rendererClass=&NiagaraMeshRendererProperties, rendererIndex=1}
   {rendererClass=&NiagaraSpriteRendererProperties, rendererIndex=2}
   {rendererClass=&NiagaraSpriteRendererProperties, rendererIndex=3}
  ]
 }
 Heart_Spline {
  emitterUpdateScript [
   &EmitterState()
   &SpawnPerFrame(Spawn:NiagaraBool=$System.DeathSpawn, Spawn Count=3, Use Spawn Probability=false)
  ]
  particleSpawnScript [
   &PartitionParticles(Number Of Partitions=3)
   &InitializeParticle(
    Color Mode=::NewEnumerator0
    Lifetime=3
    Mesh Scale Mode=::NewEnumerator3
    Mesh Uniform Scale=&FN_FrontEndFloatModifier(
     Gameplay=&SelectFloatFromArray(Direct Array Index=$Particles.PartitionParticles.Partition, Float Selection Array=DataInterface() {propertyValues {FloatData [2.3,1.3,0.5]}})
     Front End=&SelectFloatFromArray(Direct Array Index=$Particles.PartitionParticles.Partition, Float Selection Array=DataInterface() {propertyValues {FloatData [2,1,0.4]}})
    )
    UsePositionOffset=false
    Write Material Random=true
   ) {module=InitializeParticle001}
   /SetVariables(Particles.Position=&AddVectorToPosition(
     Position=$System.DeathPosition
     Vector=&MakeVector(
      X=0
      Y=0
      Z=&SelectFloatFromArray(Direct Array Index=$Particles.PartitionParticles.Partition, Float Selection Array=DataInterface() {propertyValues {FloatData [0,15,30]}})
     )
    )) {isSetParameters=true}
   &InitialMeshOrientation(
    Mesh Axis To Orient=&SelectVectorFromArray(
     Direct Array Index=$Particles.PartitionParticles.Partition
     Vector Selection Array=DataInterface() {propertyValues {FloatData [<-1 0 0>, <-1 0 0>, <-1 0 0>]}}
    )
    Mesh Orientation Mode=::NewEnumerator2
    Use Rotation Vector=false
   )
   &AddVelocity(Velocity=<0 0 1>, Velocity Speed Scale=&FN_FrontEndFloatModifier(Gameplay=220, Front End=100))
   &ApplyInitialForces()
  ]
  particleUpdateScript [
   &ParticleState()
   &ScaleMeshSize(Scale Factor=&VectorFromFloat(Value=&FloatFromCurve(FloatCurve=DataInterface() {propertyValues {Curve {keys [
          {interpMode=RCIM_Cubic, tangentMode=RCTM_User, time=0, value=0}
          {arriveTangent=2, interpMode=RCIM_Cubic, leaveTangent=-0.6586, tangentMode=RCTM_Break, time=0.3, value=1}
          <at 0.7 2 RCIM_Cubic>
         ]}}})))
   &Drag(
    Drag=&FloatFromCurve(
     FloatCurve=DataInterface() {propertyValues {Curve {keys [
         {interpMode=RCIM_Cubic, tangentMode=RCTM_User, time=0, value=0}
         {arriveTangent=9.269, interpMode=RCIM_Cubic, leaveTangent=-9.7228, tangentMode=RCTM_Break, time=0.4, value=1}
         {arriveTangent=-0.1416, interpMode=RCIM_Cubic, leaveTangent=-0.1416, tangentMode=RCTM_User, time=0.6, value=0}
         <at 1 0 RCIM_Cubic>
        ]}}}
     Scale Curve=4
    )
    Use Linear Drag=true
    Use Rotational Drag=true
   ) {module=Drag001}
   &SolveForcesAndVelocity(Clamp Velocity=false, Limit Acceleration=false)
   &DynamicMaterialParameters(
    Index 0 Param 1=&FloatFromCurve(FloatCurve=DataInterface() {propertyValues {Curve {keys [
         <at 0 0.9 RCIM_Cubic>
         {arriveTangent=-11.5319, interpMode=RCIM_Cubic, leaveTangent=0.3928, tangentMode=RCTM_Break, time=0.15, value=0}
         {arriveTangent=0.7189, interpMode=RCIM_Cubic, leaveTangent=5.1129, tangentMode=RCTM_Break, time=0.3, value=0}
         {arriveTangent=2.5663, interpMode=RCIM_Cubic, leaveTangent=-0.183, tangentMode=RCTM_Break, time=0.9, value=1.1}
        ]}}})
    Index 1 Param 1=2
    Index 1 Param 2=1
    Index 1 Param 3=&Multiply_Float(
     A=&SelectFloatFromArray(Direct Array Index=$Particles.PartitionParticles.Partition, Float Selection Array=DataInterface() {propertyValues {FloatData [99,44,30]}})
     B=&FloatFromCurve(FloatCurve=DataInterface() {propertyValues {Curve {keys [{interpMode=RCIM_Cubic, tangentMode=RCTM_User, time=0, value=0.3}, {interpMode=RCIM_Cubic, tangentMode=RCTM_User, time=0.3, value=1}]}}})
    )
    Index 1 Param 4=3
    Module.Index0Param0Write=true
    Module.Index0Param1Write=true
    Module.Index0Param2Write=true
    Module.Index0Param3Write=true
    Module.Index1Param0Write=true
    Module.Index1Param1Write=true
    Module.Index1Param2Write=true
    Module.Index1Param3Write=true
    Write Parameter Index 1=true
   ) {module=DynamicMaterialParameters003}
   &ScaleColor(
    Scale RGB=&MakeVectorFromLinearColorRGB(Color=&ColorFromCurve(ColorCurve=DataInterface() {propertyValues {
        AlphaCurve {keys [<at 0 1 RCIM_Cubic>,<at 1 1 RCIM_Cubic>]}
        BlueCurve {keys [<at 0.1069 1 RCIM_Cubic>,<at 0.337 0.3549>]}
        GreenCurve {keys [<at 0.1069 1 RCIM_Cubic>,<at 0.337 0.1487>]}
        RedCurve {keys [<at 0.1069 1 RCIM_Cubic>,<at 0.337 0.2341>]}
       }}))
    ScaleA=true
    ScaleRGB=true
   ) {module=ScaleColor001}
  ]
  rendererClasses [&NiagaraMeshRendererProperties]
  renderers [{rendererClass=&NiagaraMeshRendererProperties, rendererIndex=0}]
 }
 Hearts {
  emitterUpdateScript [
   &EmitterState()
   &SpawnPerFrame(
    Spawn:NiagaraBool=$System.DeathSpawn
    Spawn Count:NiagaraInt32=&RandomRangeInt(Minimum=3, Maximum=5)
    Use Spawn Probability=false
   )
  ]
  particleSpawnScript [
   &InitializeParticle(
    Color Maximum=<rgba 0.9928 0.0647 0.353 1>
    Color Minimum=<rgba 1 0 0.0615 1>
    Color Mode=::NewEnumerator2
    Lifetime Max=2.9
    Lifetime Min=1.5
    Lifetime Mode=::NewEnumerator1
    Sprite Size Max=<25 35>
    Sprite Size Min=<15 15>
    Sprite Size Mode=::NewEnumerator2
    UsePositionOffset=false
    Write Material Random=true
   ) {module=InitializeParticle001}
   /SetVariables(Particles.Position=$System.DeathPosition) {isSetParameters=true}
   &ShapeLocation(MaskSpawnGroup=false, Override Seed=false, Sphere Radius=9, UseOwnerScale=true) {module=ShapeLocation001}
   &AddVelocity(
    Module.Use Falloff Distance=false
    Override Seed=false
    Velocity Mode=::NewEnumerator1
    Velocity Origin=$System.DeathPosition
    Velocity Speed=&RandomRangeFloat(Minimum=333, Maximum=444)
   ) {module=AddVelocity001}
   &ApplyInitialForces()
  ]
  particleUpdateScript [
   &ParticleState()
   &SpriteRotationRate(Rotation Rate=&RandomRangeFloat(Minimum=-90, Maximum=90))
   &ScaleSpriteSize(Uniform Curve Sprite Scale=DataInterface() {propertyValues {Curve {keys [
        {interpMode=RCIM_Cubic, tangentMode=RCTM_User, time=0, value=0}
        {arriveTangent=3.3333, interpMode=RCIM_Cubic, leaveTangent=3.3333, time=0.2, value=2}
        {interpMode=RCIM_Cubic, tangentMode=RCTM_User, time=0.3, value=1}
        {tangentMode=RCTM_User, time=1, value=0}
       ]}}}) {module=ScaleSpriteSize001}
   &AccelerationForce(Acceleration=<0 0 333>)
   &Drag(
    Drag=&FloatFromCurve(
     FloatCurve=DataInterface() {propertyValues {Curve {keys [
         {arriveTangent=2.492, interpMode=RCIM_Cubic, leaveTangent=2.492, tangentMode=RCTM_User, time=0, value=0}
         {interpMode=RCIM_Cubic, tangentMode=RCTM_User, time=1, value=1}
        ]}}}
     Scale Curve=4
    )
    Use Linear Drag=true
    Use Rotational Drag=true
   )
   &CurlNoiseForce(Bias Noise=false, Noise Frequency=40.3446, Noise Strength=11.6461, Pan Noise=true, Pan Noise Field=<11 0 0>)
   &SolveForcesAndVelocity(Clamp Velocity=false, Limit Acceleration=false)
   &ScaleColor(
    Scale Alpha=&FloatFromCurve(
     FloatCurve=DataInterface() {propertyValues {Curve {keys [
         <at 0 0>
         <at 0.0805 1.8557>
         {arriveTangent=-2.0136, interpMode=RCIM_Cubic, leaveTangent=-12.6391, tangentMode=RCTM_Break, time=0.608, value=1.86}
         <at 1.0011 0.002>
        ]}}}
     Scale Curve=&RandomRangeFloat(Minimum=0.4, Maximum=0.7)
    )
    Scale RGB=&MakeVector(
     X=&FloatFromCurve(FloatCurve=DataInterface() {propertyValues {Curve {keys [
          <at 0 22 RCIM_Cubic>
          {arriveTangent=-21, interpMode=RCIM_Cubic, leaveTangent=-0.2565, tangentMode=RCTM_Break, time=0.2, value=1}
          <at 1 1 RCIM_Cubic>
         ]}}})
     Y=1
     Z=1
    )
    ScaleA=true
    ScaleRGB=true
   ) {module=ScaleColor001}
   &CameraOffset(Camera Offset Amount=1)
  ]
  rendererClasses [&NiagaraSpriteRendererProperties]
  renderers [{rendererClass=&NiagaraSpriteRendererProperties}]
 }
 Hearts_Trail {
  emitterUpdateScript [
   &EmitterState()
   &SpawnParticlesFromOtherEmitter(
    Attribute Reader=DataInterface() {propertyValues {EmitterBinding {bindingMode=Other, emitterName=Bomb_Fire001}}}
    Maximum Age Spawn Threshold=1.2
    Spawn Rate=&RandomRangeFloat(Minimum=7, Maximum=15)
    Use Maximum Age=false
    Use Minimum Age=false
   )
  ]
  particleSpawnScript [
   &InitializeParticle(
    Color Maximum=<rgba 0.9928 0.0647 0.353 1>
    Color Minimum=<rgba 1 0 0.0615 1>
    Color Mode=::NewEnumerator2
    Lifetime=&RandomRangeFloat(Minimum=0.5, Maximum=1.3)
    Sprite Size=&Multiply_Vector2D_ByFloat(Float=0.5, Vector2D=&RandomRangeVector2D(Minimum=<15 15>, Maximum=<25 35>))
    Sprite Size Mode=::NewEnumerator4
    UsePositionOffset=false
    Write Material Random=true
   ) {module=InitializeParticle001}
   &SampleParticlesFromOtherEmitter()
   &ShapeLocation(MaskSpawnGroup=false, Non Uniform Scale=<1 1 0.5>, Override Seed=false, Sphere Radius=10, UseOwnerScale=true)
   &AddVelocity(Velocity=&RandomRangeVector(Minimum=<0 0 1>, Maximum=<0.5 0.5 1>))
  ]
  particleUpdateScript [
   &ParticleState()
   &ScaleVelocity(Velocity Scale=&MakeVector(
     X=&FloatFromCurve(
      FloatCurve=DataInterface() {propertyValues {Curve {keys [
          {arriveTangent=3.4984, interpMode=RCIM_Cubic, leaveTangent=3.4984, tangentMode=RCTM_User, time=0, value=0}
          {interpMode=RCIM_Cubic, tangentMode=RCTM_User, time=1, value=1}
         ]}}}
      Scale Curve=&RandomRangeFloat(Minimum=-200, Maximum=200)
     )
     Y=0
     Z=&FloatFromCurve(
      FloatCurve=DataInterface() {propertyValues {Curve {keys [
          {arriveTangent=3.4984, interpMode=RCIM_Cubic, leaveTangent=3.4984, tangentMode=RCTM_User, time=0, value=0}
          {interpMode=RCIM_Cubic, tangentMode=RCTM_User, time=1, value=1}
         ]}}}
      Scale Curve=&RandomRangeFloat(Minimum=1, Maximum=500)
     )
    ))
   &SpriteRotationRate(Rotation Rate=&RandomRangeFloat(Minimum=-180, Maximum=180))
   &ScaleSpriteSize(Uniform Curve Sprite Scale=DataInterface() {propertyValues {Curve {keys [
        {interpMode=RCIM_Cubic, tangentMode=RCTM_User, time=0, value=0}
        {interpMode=RCIM_Cubic, tangentMode=RCTM_User, time=0.3, value=1}
        <at 1 0 RCIM_Cubic>
       ]}}})
   &CurlNoiseForce(Bias Noise=false, Noise Frequency=9.9568, Noise Strength=550.4498, Pan Noise=true, Pan Noise Field=<0 0 1>)
   &Drag(
    Drag=&FloatFromCurve(
     FloatCurve=DataInterface() {propertyValues {Curve {keys [
         {arriveTangent=-3.3118, interpMode=RCIM_Cubic, leaveTangent=-3.3118, tangentMode=RCTM_User, time=0, value=1}
         {interpMode=RCIM_Cubic, tangentMode=RCTM_User, time=1, value=0}
        ]}}}
     Scale Curve=13
    )
    Use Linear Drag=true
    Use Rotational Drag=true
   )
   &SolveForcesAndVelocity(Clamp Velocity=false, Limit Acceleration=false)
   &ScaleColor(
    Scale Alpha=&FloatFromCurve(
     FloatCurve=DataInterface() {propertyValues {Curve {keys [
         <at 0 0>
         <at 0.0805 1.8557>
         {arriveTangent=-2.0136, interpMode=RCIM_Cubic, leaveTangent=-12.6391, tangentMode=RCTM_Break, time=0.608, value=1.86}
         <at 1.0011 0.002>
        ]}}}
     Scale Curve=&RandomRangeFloat(Minimum=0.2, Maximum=0.7)
    )
    ScaleA=true
    ScaleRGB=false
   ) {module=ScaleColor001}
   &CameraOffset(Camera Offset Amount=&RandomRangeFloat(Minimum=1, Maximum=5))
  ]
  rendererClasses [&NiagaraSpriteRendererProperties]
  renderers [{rendererClass=&NiagaraSpriteRendererProperties}]
 }
 Smoke {
  emitterUpdateScript [
   &EmitterState()
   &SpawnPerFrame(
    Spawn:NiagaraBool=$System.DeathSpawn
    Spawn Count:NiagaraInt32=&RandomRangeInt(Minimum=7, Maximum=10)
    Use Spawn Probability=false
   )
  ]
  particleSpawnScript [
   &InitializeParticle(
    Color Mode=::NewEnumerator0
    Lifetime Max=1.7
    Lifetime Min=1.5
    Lifetime Mode=::NewEnumerator1
    Mass Mode=::NewEnumerator1
    Mesh Scale Mode=::NewEnumerator3
    Mesh Uniform Scale=&Multiply_Float(A=&RandomRangeFloat(Minimum=0.7, Maximum=1), B=1)
    Sprite Rotation Mode=::NewEnumerator2
    UsePositionOffset=false
    Write Material Random=true
   ) {module=InitializeParticle002}
   /SetVariables(Particles.InitSpriteRotRate=&UniformRangedFloat(Minimum=-3, Maximum=3)) {isSetParameters=true}
   &SphereLocation(
    MaskSpawnGroup=false
    Override Seed=true
    Random Seed=20
    Sphere Radius=5
    Surface Only=false
    UseNonUniformScale=false
    UseOffset=false
    UseSphereOrientation=false
   )
   &AddVelocityFromPoint(Module.Use Falloff Distance=false, Velocity Strength=99)
   /SetVariables(Particles.Position=$System.DeathPosition) {isSetParameters=true}
   &InitialMeshOrientation(
    Mesh Orientation Mode=::NewEnumerator7
    Rotation=&RandomRangeVector(Minimum=<0 0 0>, Maximum=<1 1 1>)
    Use Rotation Vector=true
   )
  ]
  particleUpdateScript [
   &ParticleState()
   &UpdateMeshOrientation(Rotation Rate=0.5, Rotation Vector=&RandomRangeVector(Minimum=<1 0 0>, Maximum=<1 0 0>))
   &AccelerationForce(Acceleration=&RandomRangeVector(Minimum=<0 0 222>, Maximum=<0 0 666>))
   &Drag(
    Drag=&FloatFromCurve(
     FloatCurve=DataInterface() {propertyValues {Curve {keys [
         {arriveTangent=0.7001, interpMode=RCIM_Cubic, leaveTangent=0.7001, tangentMode=RCTM_User, time=0, value=0}
         {interpMode=RCIM_Cubic, tangentMode=RCTM_User, time=1, value=1}
        ]}}}
     Scale Curve=4
    )
    Use Linear Drag=true
    Use Rotational Drag=true
   ) {module=Drag002}
   &SolveForcesAndVelocity(Clamp Velocity=false, Limit Acceleration=false)
   &DynamicMaterialParameters(
    Index 0 Param 1=1
    Index 0 Param 2=&FloatFromCurve(FloatCurve=DataInterface() {propertyValues {Curve {keys [
         {arriveTangent=-0.7948, interpMode=RCIM_Cubic, leaveTangent=-0.7949, tangentMode=RCTM_User, time=0.2519, value=0.1057}
         <at 0.7709 0 RCIM_Cubic>
        ]}}})
    Index 0 Param 4=&FloatFromCurve(FloatCurve=DataInterface() {propertyValues {Curve {keys [<at 0 6.7593 RCIM_Cubic>,<at 0.7 0 RCIM_Cubic>]}}})
    Module.Param0WriteEnabled=true
    Module.Param1WriteEnabled=false
    Module.Param2WriteEnabled=true
    Module.Param3WriteEnabled=true
   )
   &ScaleColor(
    Scale Alpha=&FloatFromCurve(FloatCurve=DataInterface() {propertyValues {Curve {keys [
         <at -0.0046 0.8475 RCIM_Cubic>
         {arriveTangent=-2.9221, interpMode=RCIM_Cubic, leaveTangent=-2.9221, tangentMode=RCTM_User, time=1.0011, value=-0.5}
        ]}}})
    Scale RGB=&MakeVectorFromLinearColorRGB(Color=&ColorFromCurve(ColorCurve=DataInterface() {propertyValues {
        AlphaCurve {keys [<at 0 1 RCIM_Cubic>,<at 1 1 RCIM_Cubic>]}
        BlueCurve {keys [<at 0 1.269 RCIM_Cubic>,<at 0.1449 1.269>,<at 0.3788 0.4039>,<at 0.4996 0.1245>]}
        GreenCurve {keys [<at 0 0.0961 RCIM_Cubic>,<at 0.1449 0.0961>,<at 0.3788 0>,<at 0.4996 0.0424>]}
        RedCurve {keys [<at 0 3 RCIM_Cubic>,<at 0.1449 3>,<at 0.3788 1>,<at 0.4996 0.0764>]}
       }}))
    ScaleA=true
    ScaleRGB=true
   )
   &ScaleMeshSize(Scale Factor=&VectorFromFloat(Value=&FloatFromCurve(FloatCurve=DataInterface() {propertyValues {Curve {keys [<at 0.001 0.5>,<at 0.2948 1.1652>,<at 1.0009 0.8072>]}}})))
  ]
  rendererClasses [&NiagaraMeshRendererProperties]
  renderers [{rendererClass=&NiagaraMeshRendererProperties, rendererIndex=0}]
 }
 Smoke_Ribbon {
  emitterUpdateScript [
   &EmitterState(Life Cycle Mode=::NewEnumerator1, Loop Duration=3, UseLoopDelay=true)
   &SpawnParticlesFromOtherEmitter(
    Attribute Reader=DataInterface() {propertyValues {EmitterBinding {bindingMode=Other, emitterName=Bomb_Fire001}}}
    Maximum Age Spawn Threshold=1.2
    Minimum Age Spawn Threshold=0.05
    Spawn Rate=45
    Use Maximum Age=false
    Use Minimum Age=false
   )
  ]
  particleSpawnScript [
   &InitializeParticle(
    Color Mode=::NewEnumerator0
    Lifetime=&FN_FrontEndFloatModifier(Gameplay=0.25, Front End=0.25)
    Ribbon Width=100
    Ribbon Width Mode=::NewEnumerator1
    UsePositionOffset=false
    Write Material Random=true
   )
   &SampleParticlesFromOtherEmitter(
    Age Sampling:ENiagara_DirectReadSamplingMode=::NewEnumerator1
    Mesh Orientation Sampling:ENiagara_DirectReadSamplingMode=::NewEnumerator0
    Particle ID Sampling=::NewEnumerator1
    Velocity Sampling=::NewEnumerator0
   )
   /SetVariables(Particles.Position=&AddVectorToPosition(
     Position=$Particles.Position
     Vector=&Multiply_VectorByFloat(
      Vector=&NormalizeVector(Vector To Normalize=$Particles.Velocity, Safe Normalize=true, Safe Fallback Threshold=0, Safe Fallback Vector=<0 0 0.001>)
      Float=-40
     )
    )) {isSetParameters=true}
   &InitialMeshOrientation(Mesh Orientation Mode=::NewEnumerator2, Use Rotation Vector=false)
   &ShapeLocation(MaskSpawnGroup=false, Non Uniform Scale=<1 1 0.5>, Override Seed=false, Sphere Radius=1, UseOwnerScale=true)
  ]
  particleUpdateScript [
   &SolveForcesAndVelocity(Clamp Velocity=false, Limit Acceleration=false)
   &self:SetSpawnedAge(
    Particle Attribute Reader=DataInterface() {propertyValues {EmitterBinding {bindingMode=Other, emitterName=Bomb_Fire001}}}
    Particle ID:NiagaraID=$Particles.RibbonID
   ) {module=SetSpawnedAge}
   &ParticleState()
   &ScaleColor(
    Scale Alpha=&Multiply_Float(
     A=&FloatFromCurve(
      FloatCurve=DataInterface() {propertyValues {Curve {keys [
          {interpMode=RCIM_Cubic, leaveTangent=-0.3793, tangentMode=RCTM_Break, time=0.4, value=1}
          {arriveTangent=0.2669, interpMode=RCIM_Cubic, leaveTangent=-3.6779, tangentMode=RCTM_Break, time=0.8, value=0}
         ]}}}
      CurveIndex=$Particles.SpawnNormalizedAge
      Scale Curve=0.99
     )
     B=&FloatFromCurve(
      FloatCurve=DataInterface() {propertyValues {Curve {keys [
          <at 0 1>
          {interpMode=RCIM_Cubic, leaveTangent=-5.1926, tangentMode=RCTM_Break, time=0.4, value=1}
          {interpMode=RCIM_Cubic, tangentMode=RCTM_Break, time=0.8, value=0}
         ]}}}
      Scale Curve=0.99
     )
    )
    Scale RGB=&MakeVectorFromLinearColorRGB(Color=&ColorFromCurve(ColorCurve=DataInterface() {propertyValues {
        AlphaCurve {keys [<at 0 1 RCIM_Cubic>]}
        BlueCurve {keys [<at 0.0088 0>,<at 0.6353 1.5076 RCIM_Cubic>]}
        GreenCurve {keys [<at 0.0088 1.75>,<at 0.6353 0.4293 RCIM_Cubic>]}
        RedCurve {keys [<at 0.0088 3>,<at 0.6353 2 RCIM_Cubic>]}
       }}))
    ScaleA=true
    ScaleRGB=true
   )
   &ScaleRibbonWidth(Ribbon Width Scale=&FloatFromCurve(FloatCurve=DataInterface() {propertyValues {Curve {keys [{interpMode=RCIM_Cubic, tangentMode=RCTM_User, time=0, value=0}, {interpMode=RCIM_Cubic, tangentMode=RCTM_User, time=1, value=0.5}]}}})) {module=ScaleRibbonWidth001}
  ]
  rendererClasses [&NiagaraRibbonRendererProperties]
  renderers [{rendererClass=&NiagaraRibbonRendererProperties}]
 }
 Text_KA_BOOM {
  emitterUpdateScript [
   &EmitterState()
   &SpawnPerFrame(Spawn:NiagaraBool=$System.DeathSpawn, Spawn Count=2, Use Spawn Probability=false)
  ]
  particleSpawnScript [
   &PartitionParticles()
   &InitializeParticle(
    Color Mode=::NewEnumerator0
    Lifetime=2
    Mesh Scale Mode=::NewEnumerator3
    Mesh Uniform Scale=0.25
    UsePositionOffset=false
    Write Material Random=true
   ) {module=InitializeParticle001}
   /SetVariables(Particles.Position=$System.DeathPosition) {isSetParameters=true}
   &AddVelocity(
    Velocity=<0 0 1>
    Velocity Speed Scale=&Lerp_Float(
     A=&FN_FrontEndFloatModifier(Gameplay=190, Front End=130)
     B=&FN_FrontEndFloatModifier(Gameplay=210, Front End=150)
     Alpha=&MakeFloatFromInt(Int=$Particles.PartitionParticles.Partition)
    )
   )
   &ApplyInitialForces()
  ]
  particleUpdateScript [
   &ParticleState()
   &ScaleVelocity()
   &DynamicMaterialParameters(
    Index 0 Param 1=&FloatFromCurve(FloatCurve=DataInterface() {propertyValues {Curve {keys [
         {arriveTangent=3.8817, interpMode=RCIM_Cubic, leaveTangent=3.8817, tangentMode=RCTM_User, time=0, value=0}
         {arriveTangent=0.7221, interpMode=RCIM_Cubic, leaveTangent=0.7221, tangentMode=RCTM_User, time=0.7564, value=2.091}
         {arriveTangent=4.2941, interpMode=RCIM_Cubic, leaveTangent=4.2941, tangentMode=RCTM_User, time=0.9228, value=2.078}
        ]}}})
    Index 0 Param 2=&Lerp_Float(A=1, B=0, Alpha=&MakeFloatFromInt(Int=$Particles.PartitionParticles.Partition))
    Index 0 Param 3=&Lerp_Float(
     A=&FloatFromCurve(FloatCurve=DataInterface() {propertyValues {Curve {keys [
          {arriveTangent=-17.7515, interpMode=RCIM_Cubic, leaveTangent=-17.7515, tangentMode=RCTM_User, time=-0.0046, value=1.0113}
          {arriveTangent=-0.2845, interpMode=RCIM_Cubic, leaveTangent=-0.2845, tangentMode=RCTM_User, time=0.1, value=0.065}
          {arriveTangent=0.4619, interpMode=RCIM_Cubic, leaveTangent=0.4619, tangentMode=RCTM_User, time=0.6513, value=0.0607}
          {arriveTangent=3.5589, interpMode=RCIM_Cubic, leaveTangent=3.5589, tangentMode=RCTM_User, time=1, value=1}
         ]}}})
     B=&FloatFromCurve(FloatCurve=DataInterface() {propertyValues {Curve {keys [
          {arriveTangent=-6.3759, interpMode=RCIM_Cubic, leaveTangent=-6.3759, tangentMode=RCTM_User, time=-0.0046, value=1.0113}
          {arriveTangent=-0.0113, interpMode=RCIM_Cubic, leaveTangent=-0.0113, time=0.4003, value=0.002}
          {arriveTangent=0.0913, interpMode=RCIM_Cubic, leaveTangent=0.0913, tangentMode=RCTM_User, time=1, value=1}
         ]}}})
     Alpha=&MakeFloatFromInt(Int=$Particles.PartitionParticles.Partition)
    )
    Index 0 Param 4=&Lerp_Float(A=150, B=80, Alpha=&MakeFloatFromInt(Int=$Particles.PartitionParticles.Partition))
    Module.Param0WriteEnabled=true
    Module.Param1WriteEnabled=true
    Module.Param2WriteEnabled=true
    Module.Param3WriteEnabled=true
   )
   &Drag(
    Drag=&Lerp_Float(
     A=&FloatFromCurve(
      FloatCurve=DataInterface() {propertyValues {Curve {keys [
          {interpMode=RCIM_Cubic, leaveTangent=4.8694, tangentMode=RCTM_Break, time=0.3, value=0}
          {interpMode=RCIM_Cubic, tangentMode=RCTM_User, time=1, value=5}
         ]}}}
      Scale Curve=18
     )
     B=&FloatFromCurve(
      FloatCurve=DataInterface() {propertyValues {Curve {keys [
          {interpMode=RCIM_Cubic, leaveTangent=4.8694, tangentMode=RCTM_Break, time=0.3, value=0}
          {interpMode=RCIM_Cubic, tangentMode=RCTM_User, time=1, value=5}
         ]}}}
      Scale Curve=15
     )
     Alpha=&MakeFloatFromInt(Int=$Particles.PartitionParticles.Partition)
    )
    Use Linear Drag=true
    Use Rotational Drag=true
   ) {module=Drag001}
   &SolveForcesAndVelocity(Clamp Velocity=false, Limit Acceleration=false)
   &CalculateAccurateVelocity()
   &ScaleMeshSize(Scale Factor=&Lerp_Vector(
     A=/Niagara/DynamicInputs/ScaleByCurve/ScaleVectorByCurve(
      InitialValue:Vector3f=<1 1 1>
      ScaleCurve:NiagaraDataInterfaceVectorCurve=DataInterface() {propertyValues {
        XCurve {keys [
          <at 0 0 RCIM_Cubic>
          {arriveTangent=2.2843, interpMode=RCIM_Cubic, leaveTangent=2.2843, time=0.3178, value=1.2}
          {arriveTangent=-0.4627, interpMode=RCIM_Cubic, leaveTangent=-0.4627, time=0.4378, value=1}
          {arriveTangent=-1.7786, interpMode=RCIM_Cubic, leaveTangent=-1.7786, time=0.75, value=1}
          <at 1 0 RCIM_Cubic>
         ]}
        YCurve {keys [
          <at 0 0 RCIM_Cubic>
          {arriveTangent=2.2886, interpMode=RCIM_Cubic, leaveTangent=2.2886, time=0.3167, value=1.4998}
          {arriveTangent=-1.1535, interpMode=RCIM_Cubic, leaveTangent=-1.1535, time=0.5159, value=1.1808}
          {arriveTangent=-2.4394, interpMode=RCIM_Cubic, leaveTangent=-2.4394, time=0.75, value=1}
          <at 1 0 RCIM_Cubic>
         ]}
        ZCurve {keys [
          <at 0 0 RCIM_Cubic>
          {arriveTangent=3.0427, interpMode=RCIM_Cubic, leaveTangent=3.0427, time=0.25, value=1.3258}
          {arriveTangent=-0.4, interpMode=RCIM_Cubic, leaveTangent=-0.4, time=0.37, value=1.1258}
          {arriveTangent=-1.787, interpMode=RCIM_Cubic, leaveTangent=-1.787, time=0.75, value=1.1258}
          <at 1 0 RCIM_Cubic>
         ]}
       }}
      CurveIndex:NiagaraFloat=$Particles.NormalizedAge
     )
     B=&Multiply_Vector(
      A=&ScaleVectorByCurve(
       InitialValue:Vector3f=<1 1 1>
       ScaleCurve:NiagaraDataInterfaceVectorCurve=DataInterface() {propertyValues {
         XCurve {keys [
           <at 0 0 RCIM_Cubic>
           {arriveTangent=2.2843, interpMode=RCIM_Cubic, leaveTangent=2.2843, time=0.3178, value=1.2}
           {arriveTangent=-0.4851, interpMode=RCIM_Cubic, leaveTangent=-0.4851, time=0.4378, value=1}
           {arriveTangent=-1.7786, interpMode=RCIM_Cubic, leaveTangent=-1.7786, time=0.75, value=0.9903}
           <at 1 0 RCIM_Cubic>
          ]}
         YCurve {keys [
           <at 0 0 RCIM_Cubic>
           {arriveTangent=3.2407, interpMode=RCIM_Cubic, leaveTangent=3.2407, time=0.25, value=1.3991}
           {arriveTangent=-0.6628, interpMode=RCIM_Cubic, leaveTangent=-0.6628, time=0.37, value=1.1991}
           {arriveTangent=-1.9033, interpMode=RCIM_Cubic, leaveTangent=-1.9033, time=0.75, value=1.0676}
           <at 1 0 RCIM_Cubic>
          ]}
         ZCurve {keys [
           <at 0 0 RCIM_Cubic>
           {arriveTangent=3.2407, interpMode=RCIM_Cubic, leaveTangent=3.2407, time=0.25, value=1.3991}
           {arriveTangent=-0.6628, interpMode=RCIM_Cubic, leaveTangent=-0.6628, time=0.37, value=1.1991}
           {arriveTangent=-1.9033, interpMode=RCIM_Cubic, leaveTangent=-1.9033, time=0.75, value=1.0676}
           <at 1 0 RCIM_Cubic>
          ]}
        }}
       CurveIndex:NiagaraFloat=$Particles.NormalizedAge
      )
      B=&VectorFromFloat(Value=&FloatFromCurve(FloatCurve=DataInterface() {propertyValues {Curve {keys [<at 0.75 1 RCIM_Cubic>,<at 1 0 RCIM_Cubic>]}}}))
     )
     Alpha=&MakeFloatFromInt(Int=$Particles.PartitionParticles.Partition)
    ))
   &ScaleColor(Scale RGB=&MakeVectorFromLinearColorRGB(Color=<rgba 1 1 1 1>), ScaleA=true, ScaleRGB=true)
  ]
  rendererClasses [&NiagaraMeshRendererProperties]
  renderers [{rendererClass=&NiagaraMeshRendererProperties, rendererIndex=0}]
 }
}
summary {
 systemName=NS_Contrail_BoomShot
 emitters [
  {emitterName=Bomb_Fire001, rendererClasses [&NiagaraMeshRendererProperties]}
  {emitterName=Hearts, rendererClasses [&NiagaraSpriteRendererProperties]}
  {emitterName=Hearts_Trail, rendererClasses [&NiagaraSpriteRendererProperties]}
  {emitterName=Clouds_Spline, rendererClasses [&NiagaraMeshRendererProperties]}
  {emitterName=Smoke, rendererClasses [&NiagaraMeshRendererProperties]}
  {emitterName=Smoke_Ribbon, rendererClasses [&NiagaraRibbonRendererProperties]}
  {emitterName=Heart_Spline, rendererClasses [&NiagaraMeshRendererProperties]}
  {emitterName=Text_KA_BOOM, rendererClasses [&NiagaraMeshRendererProperties]}
  {emitterName=Explosion, rendererClasses [&NiagaraMeshRendererProperties, &NiagaraSpriteRendererProperties]}
 ]
 userVariables [
  User.bIsFrontEnd {type=NiagaraBool}
  User.bIsFrontEndPreview {type=NiagaraBool, defaultValue=true}
  User.ParaGlideLeanAlpha {type=NiagaraFloat, defaultValue=0, description=""}
 ]
}
